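# M_ContrastUsingLerp is an engine material template in the PerfTricks folder that shows contrast adjustment of a texture using a lerp.
# The default diffuse texture's RGB drives the alpha of a lerp between a below-zero and an above-one constant color, stretching the range linearly.
# A Power node with a high exponent on the same RGB sits alongside, apparently as the costlier comparison.
# Neither branch shows a connection to a material output.
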Material /Engine/MaterialTemplates/PerfTricks/M_ContrastUsingLerp
graph {
 properties {pixelDepthOffsetMode=PDOM_Legacy, usedShadingModels=Default Lit}
 nodes [
  n1 = TextureSample() {texture=/Engine/EngineMaterials/DefaultDiffuse}
  LinearInterpolate(A=Constant3Vector() {constant=<rgba -0.2 -0.2 -0.2 0>}, B=Constant3Vector() {constant=<rgba 1.5 1.5 1.5 0>}, Alpha=@n1.RGB)
  Power(Base=@n1.RGB, Exp=Constant() {r=32.251})
 ]
}
