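# Engine template material in PerfTricks that derives specular colour from the diffuse texture instead of sampling a separate specular map.
# The desaturated diffuse sample is tinted by the squared RGB of the SpecularColorAlphaIntensity parameter, then scaled to a Fresnel base reflectance.
# ConvertFromDiffSpec receives this alongside the raw diffuse colour.
# The parameter's alpha is unused despite its name, and no material output connections appear.
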Material /Engine/MaterialTemplates/PerfTricks/M_SpecularUsingDiffuse
graph {
 properties {pixelDepthOffsetMode=PDOM_Legacy, usedShadingModels=Default Lit}
 nodes [
  Constant() {r=0.5}
  n1 = TextureSample() {texture=/Engine/EngineMaterials/DefaultDiffuse}
  n2 = Desaturation(None=@n1.RGB)
  n3 = VectorParameter() {parameterName=SpecularColorAlphaIntensity, defaultValue=<rgba 0.8132 0.1891 0.129 4>}
  n4 = Multiply(A=@n3.RGB, B=@n3.RGB)
  MaterialFunctionCall(DiffuseColor=@n1.RGB, SpecularColor=Multiply(A=Multiply(A=@n2, B=@n4)) {constA=1, constB=0.04, desc=FresnelBaseReflectFraction}) {materialFunction=&ConvertFromDiffSpec}
  TextureSample() {samplerType=SAMPLERTYPE_Normal, texture=/Engine/EngineMaterials/DefaultNormal}
 ]
}
