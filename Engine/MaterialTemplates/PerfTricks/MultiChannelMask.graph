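# MultiChannelMask is an engine template material in PerfTricks that demonstrates pulling several masks from each channel of one packed texture.
# T_multiMask's linear RGB feeds a remapping lerp for high values and two scaled one-minus-abs tent functions that isolate lower value bands.
# A separate red, green and blue multiply-add tint chain and a second lerp lack mask inputs.
# No node connects to a material output.
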Material /Engine/MaterialTemplates/PerfTricks/MultiChannelMask
graph {
 properties {pixelDepthOffsetMode=PDOM_Legacy, usedShadingModels=Default Lit}
 nodes [
  n1 = Constant3Vector() {constant=<rgba 1 0 0 0>}
  n2 = Constant3Vector() {constant=<rgba 0 1 0 0>}
  Add(A=Add(A=Multiply(A=@n1), B=Multiply(A=@n2)), B=Multiply(A=Constant3Vector() {constant=<rgba 0 0 1 0>}))
  LinearInterpolate(A=Constant() {r=-2}, B=Constant() {r=1.3})
  n3 = TextureSample() {samplerType=SAMPLERTYPE_LinearColor, texture=/Engine/MaterialTemplates/Textures/T_multiMask}
  LinearInterpolate(A=Constant() {r=-2}, B=Constant() {r=1}, Alpha=@n3.RGB)
  n4 = Multiply(A=Constant() {r=3.3333}, B=@n3.RGB)
  OneMinus(None=Abs(None=OneMinus(None=@n4)))
  n5 = Multiply(A=Constant() {r=2}, B=@n3.RGB)
  OneMinus(None=Abs(None=OneMinus(None=@n5)))
 ]
}
