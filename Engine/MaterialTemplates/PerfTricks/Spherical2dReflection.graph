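# Default Lit material template in the engine's PerfTricks set, named for a spherical 2D reflection trick.
# A half scale-and-bias, masked to RG, forms UVs for the default diffuse texture, though the multiply shows no source vector.
# The sampled RGB feeds the DiffuseColor input of ConvertFromDiffSpec.
# A default normal map sample sits unconnected, and no links to material outputs appear.
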Material /Engine/MaterialTemplates/PerfTricks/Spherical2dReflection
graph {
 properties {pixelDepthOffsetMode=PDOM_Legacy, usedShadingModels=Default Lit}
 nodes [
  n1 = Constant() {r=0.5}
  n2 = Add(A=Multiply(B=@n1), B=@n1)
  MaterialFunctionCall(DiffuseColor=TextureSample(UVs=ComponentMask(None=@n2) {g=true, r=true}) {texture=/Engine/EngineMaterials/DefaultDiffuse}.RGB) {materialFunction=&ConvertFromDiffSpec}
  TextureSample() {samplerType=SAMPLERTYPE_Normal, texture=/Engine/EngineMaterials/DefaultNormal}
 ]
}
