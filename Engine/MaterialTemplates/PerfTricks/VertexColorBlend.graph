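# Default Lit material template from the PerfTricks set that demonstrates blending two textures by mesh vertex color.
# A linear interpolation mixes the engine default diffuse and white grid textures, using vertex color as alpha instead of a mask texture.
# The result feeds the DiffuseColor input of ConvertFromDiffSpec.
# A default normal map sample is also present, with no connection shown in the graph.
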Material /Engine/MaterialTemplates/PerfTricks/VertexColorBlend
graph {
 properties {pixelDepthOffsetMode=PDOM_Legacy, usedShadingModels=Default Lit}
 nodes [
  MaterialFunctionCall(DiffuseColor=LinearInterpolate(
    A=TextureSample() {texture=/Engine/EngineMaterials/DefaultDiffuse}.RGB
    B=TextureSample() {texture=/Engine/EngineMaterials/DefaultWhiteGrid}.RGB
    Alpha=VertexColor()
   )) {materialFunction=&ConvertFromDiffSpec}
  TextureSample() {samplerType=SAMPLERTYPE_Normal, texture=/Engine/EngineMaterials/DefaultNormal}
 ]
}
