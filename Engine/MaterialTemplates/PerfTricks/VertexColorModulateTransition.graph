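# Material template in PerfTricks that blends two default textures, diffuse and white grid, using vertex color modulated by noise to break up the transition.
# A tiled noise texture drives a lerp from zero to vertex color, effectively multiplying them into the blend alpha.
# The blended color feeds ConvertFromDiffSpec as diffuse.
# A vertex-color contrast remap lerp and a default normal sample also exist, with no consumers shown.
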Material /Engine/MaterialTemplates/PerfTricks/VertexColorModulateTransition
graph {
 properties {pixelDepthOffsetMode=PDOM_Legacy, usedShadingModels=Default Lit}
 nodes [
  n1 = VertexColor()
  LinearInterpolate(A=Constant() {r=-0.5}, B=Constant() {r=1.2}, Alpha=@n1)
  n2 = TextureSample(UVs=TextureCoordinate() {uTiling=4, vTiling=4}) {texture=/Engine/MaterialTemplates/Textures/T_Noise01}
  MaterialFunctionCall(DiffuseColor=LinearInterpolate(
    A=TextureSample() {texture=/Engine/EngineMaterials/DefaultDiffuse}.RGB
    B=TextureSample() {texture=/Engine/EngineMaterials/DefaultWhiteGrid}.RGB
    Alpha=LinearInterpolate(B=@n1, Alpha=@n2.RGB)
   )) {materialFunction=&ConvertFromDiffSpec}
  TextureSample() {samplerType=SAMPLERTYPE_Normal, texture=/Engine/EngineMaterials/DefaultNormal}
 ]
}
